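; Tampered computer quest prop for the BattlePassS24 narrative that turns a nearby henchman hostile toward the player who completes its interaction.
; On quest success, the authority waits three seconds, passes a DoOnce and stores the player's pawn as instigator.
; It scans the encounter bounding box for a pawn with the henchman spawner tag, changes its attitude to Hostile, then resets the DoOnce.
; The construction script applies a configurable table mesh.
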(event OnQuestComponentSuccessEvent(ConsolidatedQuestComponent)
  (params (PlayerController :type Fort%20Player%20Controller%20Object%20Reference))
  (bind _returnvalue (Game|Player|GetPlayerPawn PlayerController))
  ;[0 0 0] On successful interaction, turn the henchman in range hostile towards the player, if they exist
  (Networking|SwitchHasAuthority
    (:Authority
      (Utilities|FlowControl|Delay 3.0)
      (Utilities|FlowControl|DoOnce false
        (:Completed
          (Utilities|IsValid _returnvalue
            (:Is%20Valid
              (Variables|Default|SetInstigatingPlayerPawn _returnvalue)
              (CallFunction|AggroHenchmaninBounds))))))))

(event Custom|Reset%20DoOnce
  ;[0 0 0] On successful interaction, turn the henchman in range hostile towards the player, if they exist
  (Utilities|FlowControl|DoOnce false
    (:Completed
      (Utilities|IsValid (Game|Player|GetPlayerPawn (stale OnQuestComponentSuccessEvent(ConsolidatedQuestComponent) PlayerController))
        (:Is%20Valid
          (Variables|Default|SetInstigatingPlayerPawn (Game|Player|GetPlayerPawn (stale OnQuestComponentSuccessEvent(ConsolidatedQuestComponent) PlayerController)))
          (CallFunction|AggroHenchmaninBounds))))))

(event Custom|Aggro%20Henchman%20In%20Bounds
  (bind _henchmannpc (Variables|Default|GetHenchmanNPC))
  ;[0 0 0] Find the henchman and aggro them.
  (bind _array_element (Utilities|Array|ForEachLoopwithBreak (Collision|GetOverlappingActors (Variables|Default|GetBoundingBoxEncounter) (Variables|Default|GetTandemClassFilter) :!class /Script/Engine.PrimitiveComponent))
    (:LoopBody
      (bind _asfort_player_pawn (Utilities|Casting|CastToFortPlayerPawn _array_element))
      (Utilities|IsValid _asfort_player_pawn
        (:Is%20Valid
          (if (GameplayTags|HasTag (GameplayTags|GetOwnedGameplayTags _asfort_player_pawn :!class /Script/GameplayTags.GameplayTagAssetInterface) (Variables|Default|GetHenchmanSpawnerDataTag) true)
            (Variables|Default|SetHenchmanNPC _asfort_player_pawn)))))
    (:Completed
      _))
  ;;[0.224 0.224 0.224] If henchman was found, turn it hostile and reset the DoOnce Loop in Quest Success for this prop.
  (Utilities|IsValid _henchmannpc
    (:Is%20Valid
      (Affiliation|ChangeAttitudeBetweenActors _henchmannpc (Variables|Default|GetInstigatingPlayerPawn) "Hostile")
      (CallFunction|ResetDoOnce))))

(fn ConstructionScript
  (bind _tablestaticmesh (Variables|TableConfig|GetTableStaticMesh))
  (|CallParentFunction)
  (Utilities|IsValid _tablestaticmesh
    (:Is%20Valid
      (Components|StaticMesh|SetStaticMesh (Variables|Default|GetTableStaticMesh_Comp) _tablestaticmesh))))
